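# Masked, lit material for Niagara mesh particles in the Sparrow pickaxe effects, apparently for spike meshes.
# A Step node compares the UV U coordinate with the Spike Mask dynamic parameter, giving a hard cutoff that each particle can move along the mesh.
# Particle color RGB feeds the shared emissive adjustment function, alongside a Base Color vector parameter that defaults to red.
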Material /BRCosmetics/Weapons/FORT_Melee/Pickaxe_Sparrow/FX/M_PickaxeSparrow_SpikeMask
graph {
 properties {bUsedWithNiagaraMeshParticles=true, blendMode=BLEND_Masked, pixelDepthOffsetMode=PDOM_Legacy, usedShadingModels=Default Lit}
 nodes [
  MaterialFunctionCall(Emissive=ParticleColor().RGB) {materialFunction=&MF_FX_EmissiveAdjustments}
  Step(
   Y=ComponentMask(None=TextureCoordinate()) {r=true}
   X=DynamicParameter() {defaultValue=<rgba 0 1 1 1>, paramNames [Spike Mask, Param2, Param3, Param4]}.Spike Mask
  ) {constX=0}
  VectorParameter() {parameterName=Base Color, defaultValue=<rgba 0.9 0 0 1>}
 ]
}
