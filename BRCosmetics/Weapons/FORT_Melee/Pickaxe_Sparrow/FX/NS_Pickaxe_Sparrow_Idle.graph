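# Niagara idle effect for the Sparrow pickaxe: bone-attached gem meshes and socket glows that react to a weak point counter.
# WeakPoint int picks a stepped material value for both gems, tints one, and drives glow alpha.
# Glow001 puts sprites on six FX sockets, toggling their size per spawn group; the older Glow emitter is disabled.
# "Set reactive full on" overrides to full, and time of day scales some colors.
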NiagaraSystem /BRCosmetics/Weapons/FORT_Melee/Pickaxe_Sparrow/FX/NS_Pickaxe_Sparrow_Idle
dependencies {
 usedDataInterfaces [
  &NiagaraDataInterfaceArrayFloat
  &NiagaraDataInterfaceSkeletalMesh
  &NiagaraDataInterfaceArrayInt32
  &NiagaraDataInterfaceCurve
 ]
 usedDynamicInputs [&FNTimeOfDayScalar]
 usedModules [
  &SystemState
  &InterpolateOverTime
  &self:SystemSpawnScript.SystemScriptSource.SystemScriptGraph.NiagaraNodeAssignment_0.SetVariables_#1
 ]
}
emitters {
 Gem {
  emitterUpdateScript [
   &EmitterState(Life Cycle Mode=::NewEnumerator1, Loop Behavior=::NewEnumerator1, Loop Duration Mode=::NewEnumerator1, UseLoopDelay=false)
   &SpawnBurst_Instantaneous(Use Spawn Probability=false, UseLoopCountLimit=false)
  ]
  particleSpawnScript [
   &InitializeParticle(Color Mode=::NewEnumerator0, Mesh Scale=<1.02 1 1>, Mesh Scale Mode=::NewEnumerator4, UsePositionOffset=false, Write Material Random=true)
   &InitialMeshOrientation(Mesh Orientation Mode=::NewEnumerator1, Rotation=<0 0 0.5>, Use Rotation Vector=true)
  ]
  particleUpdateScript [
   &DynamicMaterialParameters(
    Index 0 Param 1=&Lerp_Float(
     A=&SelectFloatFromArray(Direct Array Index=$User.WeakPoint int, Float Selection Array=DataInterface() {propertyValues {FloatData [0,0.16,0.32,0.48,0.63,1]}})
     B=1
     Alpha=$User.Set reactive full on
    )
    Module.Param0WriteEnabled=true
    Module.Param1WriteEnabled=true
    Module.Param2WriteEnabled=true
    Module.Param3WriteEnabled=true
   )
   &SkeletalMeshLocation(
    Bone Sampling Mode=::NewEnumerator4
    MaskSpawnGroup=false
    Sampled Position Offset=<0.01 0 0>
    Skeletal Mesh=DataInterface() {propertyValues {
      DefaultMesh=null
      ExcludeBoneName=root
      OptionalComponentClass=null
      PreviewMesh=/BRCosmetics/Weapons/FORT_Melee/Pickaxe_Sparrow/Meshes/Sparrow_Axe
      SkinningMode=SkinOnTheFly
      SoftSourceActor=null
      SourceMode=Default
      UvSetIndex=0
      WholeMeshLOD=-1
      bExcludeBone=false
      bReadDeformedGeometry=true
      bRequireCurrentFrameData=true
      ComponentTags []
      FilteredBones [b_AR_Root]
      FilteredSockets []
      MeshUserParameter {parameter {name=None, typeDefHandle {registeredTypeIndex=1}, varData []}}
      SamplingRegions []
     }}
   )
   &ParticleState(Kill Particles When Lifetime Has Elapsed=false, Let Infinitely Lived Particles Die When Emitter Deactivates=true)
   &Color(Color=&MakeCustomLinearColorFromBool(
     Bool=&SetBoolByIntComparison(A=$User.WeakPoint int, B=5, Comparison Type=::NewEnumerator2)
     True Color=<rgba 0.2 0 0 1>
     False Color=<rgba 0.1 0 0.0028 1>
    ))
   &ScaleColor(Scale RGB=&Multiply_VectorByFloat(Vector=<1 1 1>, Float=$System.TOD), ScaleA=true, ScaleRGB=true) {enabled=false}
   &CameraOffset(Camera Offset Amount=0.05)
  ]
  rendererClasses [&NiagaraMeshRendererProperties]
  renderers [{rendererClass=&NiagaraMeshRendererProperties, rendererIndex=0}]
 }
 Gem001 {
  emitterUpdateScript [
   &EmitterState(Life Cycle Mode=::NewEnumerator1, Loop Behavior=::NewEnumerator1, Loop Duration Mode=::NewEnumerator1, UseLoopDelay=false)
   &SpawnBurst_Instantaneous(Use Spawn Probability=false, UseLoopCountLimit=false)
  ]
  particleSpawnScript [
   &InitializeParticle(Color=<rgba 0.2 0 0 1>, Mesh Scale=<1.04 1 1>, Mesh Scale Mode=::NewEnumerator4, UsePositionOffset=false, Write Material Random=true)
   &InitialMeshOrientation(Mesh Orientation Mode=::NewEnumerator1, Rotation=<0 0.25 0>, Use Rotation Vector=true)
  ]
  particleUpdateScript [
   &DynamicMaterialParameters(
    Index 0 Param 1=&Lerp_Float(
     A=&SelectFloatFromArray(Direct Array Index=$User.WeakPoint int, Float Selection Array=DataInterface() {propertyValues {FloatData [0,0.16,0.32,0.48,0.62,1]}})
     B=1
     Alpha=$User.Set reactive full on
    )
    Module.Param0WriteEnabled=true
    Module.Param1WriteEnabled=true
    Module.Param2WriteEnabled=true
    Module.Param3WriteEnabled=true
   )
   &SkeletalMeshLocation(
    Bone Sampling Mode=::NewEnumerator4
    MaskSpawnGroup=false
    Sampled Position Offset=<-10.3 -0.035 0.01>
    Skeletal Mesh=DataInterface() {propertyValues {
      DefaultMesh=null
      ExcludeBoneName=root
      OptionalComponentClass=null
      PreviewMesh=&Sparrow_Axe
      SkinningMode=SkinOnTheFly
      SoftSourceActor=null
      SourceMode=Default
      UvSetIndex=0
      WholeMeshLOD=-1
      bExcludeBone=false
      bReadDeformedGeometry=true
      bRequireCurrentFrameData=true
      ComponentTags []
      FilteredBones [C_tip]
      FilteredSockets []
      MeshUserParameter {parameter {name=None, typeDefHandle {registeredTypeIndex=1}, varData []}}
      SamplingRegions []
     }}
   )
   &ParticleState(Kill Particles When Lifetime Has Elapsed=false, Let Infinitely Lived Particles Die When Emitter Deactivates=true)
   &ScaleColor(Scale RGB=&Multiply_VectorByFloat(Vector=<1 1 1>, Float=$System.TOD), ScaleA=true, ScaleRGB=true)
   &CameraOffset(Camera Offset Amount=0.05)
  ]
  rendererClasses [&NiagaraMeshRendererProperties]
  renderers [{rendererClass=&NiagaraMeshRendererProperties, rendererIndex=0}]
 }
 Glow {
  bEnabled=false
  emitterUpdateScript [
   &EmitterState(Life Cycle Mode=::NewEnumerator1, Loop Behavior=::NewEnumerator1, Loop Duration Mode=::NewEnumerator1, UseLoopDelay=false)
   &SpawnBurst_Instantaneous(Spawn Count=6, Use Spawn Probability=false, UseLoopCountLimit=false)
  ]
  particleSpawnScript [
   &ApplyInitialForces() {enabled=false}
   &FN_SetSpawnGroup(Number of Spawn Groups=6)
   &InitializeParticle(
    Color=<rgba 1 0 0.0281 0.2>
    Sprite Size=&Multiply_Vector2D_ByFloat(
     Float=&SelectFloatFromArray(Direct Array Index=$Particles.SpawnGroup, Float Selection Array=DataInterface() {propertyValues {FloatData [40,40,40,40,40,40]}})
     Vector2D=<1 1>
    )
    Sprite Size Mode=::NewEnumerator4
    UsePositionOffset=false
    Write Material Random=true
   )
   &InitialMeshOrientation(Mesh Orientation Mode=::NewEnumerator1, Rotation=<0 0 0.5>, Use Rotation Vector=true) {enabled=false}
   &CameraOffset(Camera Offset Amount=15)
   &SpriteFacingAndAlignment(Sprite Alignment=<0 0 1>, Sprite Facing=<0 0 1>, WriteAlignment=true, WriteFacing=false) {enabled=false}
   &AlignSpriteToMeshOrientation(Mesh Orientation Relative Sprite Alignment Vector=<0 0 1>, SetAlignment=true, SetFacing=true) {enabled=false}
  ]
  particleUpdateScript [
   &SkeletalMeshLocation(
    Bone / Socket Index=&MakeCustomIntFromBool(
     Bool=&SetBoolByFloatComparison(A=$User.Set reactive full on, B=0.5)
     True INT=&ReturnParticleIDAttributes()
     False INT=&Add_Integer(
      A=$Particles.SpawnGroup
      B=&SelectIntFromArray(Direct Array Index=$User.WeakPoint int, Int Selection Array=DataInterface() {propertyValues {IntData [-5,-4,-3,-2,-1,0]}})
     )
    )
    MaskSpawnGroup=false
    Mesh Sampling Type=::NewEnumerator5
    Skeletal Mesh=DataInterface() {propertyValues {
      DefaultMesh=null
      ExcludeBoneName=root
      OptionalComponentClass=null
      PreviewMesh=&Sparrow_Axe
      SkinningMode=SkinOnTheFly
      SoftSourceActor=null
      SourceMode=Default
      UvSetIndex=0
      WholeMeshLOD=-1
      bExcludeBone=false
      bReadDeformedGeometry=true
      bRequireCurrentFrameData=true
      ComponentTags []
      FilteredBones []
      FilteredSockets [FX Socket01, FX Socket02, FX Socket03, FX Socket04, FX Socket05, FX Socket_Base]
      MeshUserParameter {parameter {name=None, typeDefHandle {registeredTypeIndex=1}, varData []}}
      SamplingRegions []
     }}
    Socket Sampling Mode=::NewEnumerator5
   )
   &ParticleState(Kill Particles When Lifetime Has Elapsed=false, Let Infinitely Lived Particles Die When Emitter Deactivates=true)
   &ScaleSpriteSize(Scale Sprite Size Mode=::NewEnumerator0) {module=ScaleSpriteSize002}
   &ScaleSpriteSize(Scale Sprite Size Mode=::NewEnumerator0, Uniform Scale Factor=1.2)
   &ScaleColor(
    Scale Alpha=&Lerp_Float(
     A=&Multiply_Float(
      A=&Add_Float(
       A=0
       B=&Multiply_Float(A=&MakeFloatFromInt(Int=$User.WeakPoint int), B=0.2)
      )
      B=1
     )
     B=1
     Alpha=$User.Set reactive full on
    )
    Scale RGB=&Multiply_VectorByFloat(Vector=<1 1 1>, Float=$System.TOD)
    ScaleA=true
    ScaleRGB=true
   )
   &Color(Color=<rgba 1 1 1 5000>) {enabled=false}
   &ScaleSpriteSize(Scale Sprite Size Mode=::NewEnumerator0, Uniform Scale Factor=0.2) {enabled=false, module=ScaleSpriteSize001}
  ]
  rendererClasses [&NiagaraSpriteRendererProperties]
  renderers [{rendererClass=&NiagaraSpriteRendererProperties}]
 }
 Glow001 {
  emitterUpdateScript [
   &EmitterState(Life Cycle Mode=::NewEnumerator1, Loop Behavior=::NewEnumerator1, Loop Duration Mode=::NewEnumerator1, UseLoopDelay=false)
   &SpawnBurst_Instantaneous(Use Spawn Probability=false, UseLoopCountLimit=false)
   &SpawnBurst_Instantaneous(Spawn Group=1, Use Spawn Probability=false, UseLoopCountLimit=false) {module=SpawnBurst_Instantaneous005}
   &SpawnBurst_Instantaneous(Spawn Group=2, Use Spawn Probability=false, UseLoopCountLimit=false) {module=SpawnBurst_Instantaneous004}
   &SpawnBurst_Instantaneous(Spawn Group=3, Use Spawn Probability=false, UseLoopCountLimit=false) {module=SpawnBurst_Instantaneous003}
   &SpawnBurst_Instantaneous(Spawn Group=4, Use Spawn Probability=false, UseLoopCountLimit=false) {module=SpawnBurst_Instantaneous002}
   &SpawnBurst_Instantaneous(Spawn Group=5, Use Spawn Probability=false, UseLoopCountLimit=false) {module=SpawnBurst_Instantaneous001}
  ]
  particleSpawnScript [
   &InitializeParticle(Color=<rgba 1 0 0.0281 0.2>, Sprite Size Mode=::NewEnumerator3, Uniform Sprite Size=40, UsePositionOffset=false, Write Material Random=true)
   &CameraOffset(Camera Offset Amount=15)
   &FN_SetSpawnGroup(Number of Spawn Groups=6)
  ]
  particleUpdateScript [
   &SkeletalMeshLocation(
    Bone / Socket Index=&ReturnParticleIDAttributes()
    MaskSpawnGroup=false
    Mesh Sampling Type=::NewEnumerator5
    Skeletal Mesh=DataInterface() {propertyValues {
      DefaultMesh=null
      ExcludeBoneName=root
      OptionalComponentClass=null
      PreviewMesh=&Sparrow_Axe
      SkinningMode=SkinOnTheFly
      SoftSourceActor=null
      SourceMode=Default
      UvSetIndex=0
      WholeMeshLOD=-1
      bExcludeBone=false
      bReadDeformedGeometry=true
      bRequireCurrentFrameData=true
      ComponentTags []
      FilteredBones []
      FilteredSockets [FX Socket01, FX Socket02, FX Socket03, FX Socket04, FX Socket05, FX Socket_Base]
      MeshUserParameter {parameter {name=None, typeDefHandle {registeredTypeIndex=1}, varData []}}
      SamplingRegions []
     }}
    Socket Sampling Mode=::NewEnumerator5
   )
   &ParticleState(Kill Particles When Lifetime Has Elapsed=false, Let Infinitely Lived Particles Die When Emitter Deactivates=true)
   &ScaleSpriteSize(Scale Sprite Size Mode=::NewEnumerator0, Uniform Scale Factor=1.2)
   &ScaleSpriteSize(
    Scale Sprite Size Mode=::NewEnumerator0
    Uniform Scale Factor=&MakeCustomFloatFromBool(
     Bool=&SetBoolByFloatComparison(A=$User.Set reactive full on, B=0.5)
     True Float=1
     False Float=&MaskFloatBySpawnGroup(
      Masked Float=&MakeCustomFloatFromBool(
       Bool=&SetBoolByIntComparison(A=$User.WeakPoint int, B=0, Comparison Type=::NewEnumerator0)
       True Float=1
       False Float=0
      )
      Passthrough Float=&MaskFloatBySpawnGroup(
       Masked Float=&MakeCustomFloatFromBool(
        Bool=&SetBoolByIntComparison(A=$User.WeakPoint int, B=1, Comparison Type=::NewEnumerator0)
        True Float=1
        False Float=0
       )
       Passthrough Float=&MaskFloatBySpawnGroup(
        Masked Float=&MakeCustomFloatFromBool(
         Bool=&SetBoolByIntComparison(A=$User.WeakPoint int, B=2, Comparison Type=::NewEnumerator0)
         True Float=1
         False Float=0
        )
        Passthrough Float=&MaskFloatBySpawnGroup(
         Masked Float=&MakeCustomFloatFromBool(
          Bool=&SetBoolByIntComparison(A=$User.WeakPoint int, B=3, Comparison Type=::NewEnumerator0)
          True Float=1
          False Float=0
         )
         Passthrough Float=&MaskFloatBySpawnGroup(
          Masked Float=&MakeCustomFloatFromBool(
           Bool=&SetBoolByIntComparison(A=$User.WeakPoint int, B=4, Comparison Type=::NewEnumerator0)
           True Float=1
           False Float=0
          )
          Passthrough Float=&MakeCustomFloatFromBool(
           Bool=&SetBoolByIntComparison(A=$User.WeakPoint int, B=4, Comparison Type=::NewEnumerator0)
           True Float=1
           False Float=0
          )
          Spawn Group Mask=4
         )
         Spawn Group Mask=3
        )
        Spawn Group Mask=2
       )
       Spawn Group Mask=1
      )
     )
    )
   ) {module=ScaleSpriteSize002}
   &ScaleColor(
    Scale Alpha=&Lerp_Float(
     A=&Multiply_Float(
      A=&Add_Float(
       A=0
       B=&Multiply_Float(A=&MakeFloatFromInt(Int=$User.WeakPoint int), B=0.2)
      )
      B=1
     )
     B=1
     Alpha=$User.Set reactive full on
    )
    Scale RGB=&Multiply_VectorByFloat(Vector=<1 1 1>, Float=$System.TOD)
    ScaleA=true
    ScaleRGB=true
   )
   &Color(Color=<rgba 1 1 1 5000>) {enabled=false}
   &ScaleSpriteSize(Scale Sprite Size Mode=::NewEnumerator0, Uniform Scale Factor=0.2) {enabled=false, module=ScaleSpriteSize001}
  ]
  rendererClasses [&NiagaraSpriteRendererProperties]
  renderers [{rendererClass=&NiagaraSpriteRendererProperties}]
 }
}
summary {
 systemName=NS_Pickaxe_Sparrow_Idle
 emitters [
  {emitterName=Gem, rendererClasses [&NiagaraMeshRendererProperties]}
  {bEnabled=false, emitterName=Glow, rendererClasses [&NiagaraSpriteRendererProperties]}
  {emitterName=Gem001, rendererClasses [&NiagaraMeshRendererProperties]}
  {emitterName=Glow001, rendererClasses [&NiagaraSpriteRendererProperties]}
 ]
 userVariables [
  User.bIsFrontEnd {type=NiagaraBool}
  User.bIsFrontEndPreview {type=NiagaraBool}
  User.Set reactive full on {type=NiagaraFloat, defaultValue=0, description=""}
  User.WeakPoint int {type=NiagaraInt32}
 ]
}
