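# Animation blueprint for the Sparrow pickaxe that procedurally wobbles the C_tip bone.
# Each update builds a bone-space translation from three real-time sine waves at different frequencies, with the X wave remapped to an offset range.
# A Transform Bone node applies it over the input pose, using motion amount as alpha.
# That alpha switches to the stronger reactive amount while jumping or playing a melee animation, otherwise the subtler regular amount.
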AnimBlueprint /BRCosmetics/Weapons/FORT_Melee/Pickaxe_Sparrow/Meshes/Sparrow_AnimBlueprint
target_skeleton=&/Sparrow_Skeleton
graphs [
 AnimGraph {nodes [
   n1 = "Transform (Modify) Bone - Bone: C_tip"(
    ComponentPose=Local To Component(LocalPose=Input Pose())
    Alpha=Get motion amount() -> (motion amount:float)
    Translation=Get SpeedReactiveVar() -> (SpeedReactiveVar:Vector)
    Rotation="0,0,0"
    Scale="1,1,1"
   ) {
    translationMode=BMM_Replace
    translationSpace=BCS_BoneSpace
    alphaBoolBlend {blendOption=Linear}
    alphaScaleBiasClamp {bInterpResult=true, interpSpeedDecreasing=1, interpSpeedIncreasing=1}
    boneToModify {boneName=C_tip}
   }
   Output Pose(Result=Component To Local(ComponentPose=@n1)) {layerGroup=None}
  ]}
 EventGraph {nodes [
   Event BlueprintUpdateAnimation()
   > n1 = Sequence()
   > n2 = Branch(Condition=OR Boolean(A=Get bIsJumping(), B=Get bIsPlayingMeleeAnim()))
   > Set motion amount(motion amount:float=Get Motion Amount_reactive() -> (Motion Amount_reactive:float)) -> (Output_Get:float)
   Set motion amount(execute=@n2.else, motion amount=Get Motion Amount_regular() -> (Motion Amount_regular:float))
   n3 = GetRealTimeSeconds()
   n4 = "Sin (Radians)"(A=float * float(A=float * float(A=0.9, B=1.1), B=@n3))
   n5 = "Sin (Radians)"(A=float * float(A=float * float(A=1.1, B=1.1), B=@n3))
   n6 = "Sin (Radians)"(A=float * float(A=float * float(A=3.5, B=1.3), B=@n3))
   Set SpeedReactiveVar(
    execute=@n1.then_1
    SpeedReactiveVar:Vector=MakeVector(
     X=Remap Float(Input=@n6, OldRange=(X=-1,Y=1), NewRange=(X=-2,Y=4)) -> (Output=~)
     Y=float * float(A=@n5, B=0.15)
     Z=float * float(A=@n4, B=0.3)
    )
   ) -> (Output_Get:Vector)
   TryGetPawnOwner()
  ]}
]
variables [
 motion amount {type=number, default=1}
 Motion Amount_reactive {type=number, default=1}
 Motion Amount_regular {type=number, default=0.25}
 NewVar {type=number}
 SpeedReactiveVar {type=Vector}
 SpeedRegular {type=Vector}
]
